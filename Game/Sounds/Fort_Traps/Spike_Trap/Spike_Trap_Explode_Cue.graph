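# Sound cue for the spike trap explosion, switching its content by audio quality level.
# The first two quality slots mix a random spike-destroy sound with a quieter, slightly delayed random junk impact, the second with fewer junk variants.
# The other two slots play a single modulated, optimized explosion wave.
# It uses the trap sound class and attenuation, with overridden concurrency that prevents new instances at the limit.
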SoundCue /Game/Sounds/Fort_Traps/Spike_Trap/Spike_Trap_Explode_Cue
root=QualityLevel_0
nodes [
 {id=Delay_1, children [Random_3], properties {delayMax=0.2, delayMin=0.2}}
 {id=Delay_2, children [Random_5], properties {delayMax=0.2, delayMin=0.2}}
 {id=Mixer_1, children [Random_2,Delay_1], properties {inputVolume [1,0.5]}}
 {id=Mixer_2, children [Random_4,Delay_2], properties {inputVolume [1,0.5]}}
 {id=Modulator_0, children [WavePlayer_0]}
 {id=Modulator_1, children [WavePlayer_1]}
 {id=QualityLevel_0, children [Mixer_1,Mixer_2,Modulator_0,Modulator_1]}
 {id=Random_2, children [WavePlayer_6,WavePlayer_7], properties {weights [1,1]}}
 {id=Random_3, children [WavePlayer_8, WavePlayer_9, WavePlayer_10, WavePlayer_11], properties {weights [1,1,1,1]}}
 {id=Random_4, children [WavePlayer_12, WavePlayer_13], properties {weights [1,1]}}
 {id=Random_5, children [WavePlayer_15, WavePlayer_16, WavePlayer_17], properties {weights [1,1,1]}}
 {id=WavePlayer_0, properties {soundWaveAssetPtr=&/Spike_Trap_Explode_opt}}
 {id=WavePlayer_1, properties {soundWaveAssetPtr=&Spike_Trap_Explode_opt}}
 {id=WavePlayer_10, properties {soundWaveAssetPtr=/Game/Sounds/Fort_Impact_Sounds/Demolition/Junk/Junk_Impact_Large08}}
 {id=WavePlayer_11, properties {soundWaveAssetPtr=/Game/Sounds/Fort_Impact_Sounds/Demolition/Junk/Junk_Impact_Large09}}
 {id=WavePlayer_12, properties {soundWaveAssetPtr=&/FloorTrap_Spikes_Destroy_002}}
 {id=WavePlayer_13, properties {soundWaveAssetPtr=&/FloorTrap_Spikes_Destroy_003}}
 {id=WavePlayer_15, properties {soundWaveAssetPtr=/Game/Sounds/Fort_Impact_Sounds/Demolition/Junk/Junk_Impact_Large07}}
 {id=WavePlayer_16, properties {soundWaveAssetPtr=&Junk_Impact_Large08}}
 {id=WavePlayer_17, properties {soundWaveAssetPtr=&Junk_Impact_Large09}}
 {id=WavePlayer_6, properties {soundWaveAssetPtr=&FloorTrap_Spikes_Destroy_002}}
 {id=WavePlayer_7, properties {soundWaveAssetPtr=&FloorTrap_Spikes_Destroy_003}}
 {id=WavePlayer_8, properties {soundWaveAssetPtr=/Game/Sounds/Fort_Impact_Sounds/Demolition/Junk/Junk_Impact_Large06}}
 {id=WavePlayer_9, properties {soundWaveAssetPtr=&Junk_Impact_Large07}}
]
properties {
 attenuationSettings=&Fort_Traps_Att
 bOverrideConcurrency=true
 maxDistance=6000
 soundClassObject=&Fort_Traps
 volumeMultiplier=0.5
 attenuationOverrides {falloffDistance=3200, attenuationShapeExtents {x=800}}
 concurrencyOverrides {resolutionRule=PreventNew}
}
