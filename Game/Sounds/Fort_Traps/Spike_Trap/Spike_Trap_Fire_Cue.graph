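# Sound cue for the spike trap firing, routed to the trap placement sound class with trap attenuation.
# A quality-level node selects one of four branches, each with slight random pitch modulation.
# The first layers a random choice of two floor-spike waves with a fire wave, the second fixes that pair, and the last two play a single optimized wave.
# Overridden concurrency prevents new instances once the limit is reached.
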SoundCue /Game/Sounds/Fort_Traps/Spike_Trap/Spike_Trap_Fire_Cue
root=QualityLevel_0
nodes [
 {id=Mixer_0, children [WavePlayer_3,WavePlayer_2], properties {inputVolume [1,1]}}
 {id=Mixer_4, children [Random_13,WavePlayer_0], properties {inputVolume [1,1]}}
 {id=Modulator_0, children [WavePlayer_1], properties {pitchMax=1.1, pitchMin=0.9}}
 {id=Modulator_1, children [Mixer_0], properties {pitchMax=1.1, pitchMin=0.9}}
 {id=Modulator_2, children [WavePlayer_4], properties {pitchMax=1.1, pitchMin=0.9}}
 {id=Modulator_6, children [Mixer_4], properties {pitchMax=1.1, pitchMin=0.9}}
 {id=QualityLevel_0, children [Modulator_6,Modulator_1,Modulator_0,Modulator_2]}
 {id=Random_13, children [WavePlayer_53, WavePlayer_54], properties {weights [1,1]}}
 {id=WavePlayer_0, properties {soundWaveAssetPtr=&/Spike_Trap_Fire01}}
 {id=WavePlayer_1, properties {soundWaveAssetPtr=&/Spike_Trap_Fire_opt}}
 {id=WavePlayer_2, properties {soundWaveAssetPtr=&Spike_Trap_Fire01}}
 {id=WavePlayer_3, properties {soundWaveAssetPtr=&/FloorTrap_Spikes_Fire_003}}
 {id=WavePlayer_4, properties {soundWaveAssetPtr=&Spike_Trap_Fire_opt}}
 {id=WavePlayer_53, properties {soundWaveAssetPtr=&FloorTrap_Spikes_Fire_003}}
 {id=WavePlayer_54, properties {soundWaveAssetPtr=&/FloorTrap_Spikes_Fire_004}}
]
properties {
 attenuationSettings=&Fort_Traps_Att
 bOverrideConcurrency=true
 maxDistance=6000
 soundClassObject=/Game/Sounds/FortSoundClassesAndModes/Fort_Traps_Placement
 volumeMultiplier=0.5
 attenuationOverrides {bAttenuateWithLPF=true, falloffDistance=3100, lPFRadiusMin=2500, attenuationShapeExtents {x=900}}
 concurrencyOverrides {resolutionRule=PreventNew}
}
