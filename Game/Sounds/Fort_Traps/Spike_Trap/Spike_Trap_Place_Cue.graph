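# Sound cue for placing a spike trap, routed to the trap placement sound class with shared trap attenuation.
# A quality-level switch picks one of four branches: the first two mix a pitch-varied placement wave with a briefly delayed floor-spike layer.
# The first branch randomizes between two spike waves, the second uses one, and the last two play only a pitch-varied optimized wave.
# Overridden concurrency prevents new instances once the limit is reached.
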SoundCue /Game/Sounds/Fort_Traps/Spike_Trap/Spike_Trap_Place_Cue
root=QualityLevel_0
nodes [
 {id=Delay_0, children [WavePlayer_3], properties {delayMax=0.2, delayMin=0.2}}
 {id=Delay_2, children [Random_14], properties {delayMax=0.2, delayMin=0.2}}
 {id=Mixer_0, children [Delay_0,Modulator_1], properties {inputVolume [0.8,1.4]}}
 {id=Mixer_5, children [Delay_2,Modulator_7], properties {inputVolume [0.8,1.4]}}
 {id=Modulator_0, children [WavePlayer_0], properties {pitchMax=1.08, pitchMin=0.92}}
 {id=Modulator_1, children [WavePlayer_1], properties {pitchMax=1.3}}
 {id=Modulator_2, children [WavePlayer_5], properties {pitchMax=1.08, pitchMin=0.92}}
 {id=Modulator_7, children [WavePlayer_2], properties {pitchMax=1.3}}
 {id=QualityLevel_0, children [Mixer_5,Mixer_0,Modulator_0,Modulator_2]}
 {id=Random_14, children [WavePlayer_55, WavePlayer_56], properties {weights [1,1]}}
 {id=WavePlayer_0, properties {soundWaveAssetPtr=&/Spike_Trap_Place_opt}}
 {id=WavePlayer_1, properties {soundWaveAssetPtr=&/Spike_Trap_Place01}}
 {id=WavePlayer_2, properties {soundWaveAssetPtr=&Spike_Trap_Place01}}
 {id=WavePlayer_3, properties {soundWaveAssetPtr=&/FloorTrap_Spikes_Placed_001}}
 {id=WavePlayer_5, properties {soundWaveAssetPtr=&Spike_Trap_Place_opt}}
 {id=WavePlayer_55, properties {soundWaveAssetPtr=&FloorTrap_Spikes_Placed_001}}
 {id=WavePlayer_56, properties {soundWaveAssetPtr=&/FloorTrap_Spikes_Placed_002}}
]
properties {
 attenuationSettings=&Fort_Traps_Att
 bOverrideConcurrency=true
 maxDistance=6000
 soundClassObject=/Game/Sounds/FortSoundClassesAndModes/Fort_Traps_Placement
 volumeMultiplier=0.9
 attenuationOverrides {falloffDistance=2400, attenuationShapeExtents {x=600}}
 concurrencyOverrides {resolutionRule=PreventNew}
}
