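# Sound cue for the spike trap reload, layering a short-delayed reload sound with a later, quieter spikes-ready sound.
# A quality level node scales detail: the first tier randomizes between two ready waves, the second uses one, and the last two play only an optimized reload wave.
# It uses the trap sound class and attenuation, with concurrency overridden to prevent new instances once the limit is reached.
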SoundCue /Game/Sounds/Fort_Traps/Spike_Trap/Spike_Trap_Reload_Cue
root=QualityLevel_0
nodes [
 {id=Delay_0, children [WavePlayer_0], properties {delayMax=0.3, delayMin=0.3}}
 {id=Delay_1, children [WavePlayer_1], properties {delayMax=0.3, delayMin=0.3}}
 {id=Delay_2, children [WavePlayer_2], properties {delayMax=0.3, delayMin=0.3}}
 {id=Delay_3, children [WavePlayer_3], properties {delayMax=1.3, delayMin=1.3}}
 {id=Delay_4, children [WavePlayer_4], properties {delayMax=0.3, delayMin=0.3}}
 {id=Delay_5, children [Random_15], properties {delayMax=1.3, delayMin=1.3}}
 {id=Mixer_0, children [Delay_3,Delay_2], properties {inputVolume [0.6,1.5]}}
 {id=Mixer_6, children [Delay_5,Delay_4], properties {inputVolume [0.6,1.5]}}
 {id=QualityLevel_0, children [Mixer_6,Mixer_0,Delay_0,Delay_1]}
 {id=Random_15, children [WavePlayer_58, WavePlayer_59], properties {weights [1,1]}}
 {id=WavePlayer_0, properties {soundWaveAssetPtr=&/Spike_Trap_Reload_opt}}
 {id=WavePlayer_1, properties {soundWaveAssetPtr=&Spike_Trap_Reload_opt}}
 {id=WavePlayer_2, properties {soundWaveAssetPtr=&/Spike_Trap_Reload03}}
 {id=WavePlayer_3, properties {soundWaveAssetPtr=&/FloorTrap_Spikes_Ready_001}}
 {id=WavePlayer_4, properties {soundWaveAssetPtr=&Spike_Trap_Reload03}}
 {id=WavePlayer_58, properties {soundWaveAssetPtr=&FloorTrap_Spikes_Ready_001}}
 {id=WavePlayer_59, properties {soundWaveAssetPtr=&/FloorTrap_Spikes_Ready_002}}
]
properties {
 attenuationSettings=&Fort_Traps_Att
 bOverrideConcurrency=true
 maxDistance=6000
 soundClassObject=&Fort_Traps
 volumeMultiplier=0.5
 attenuationOverrides {falloffDistance=1900, attenuationShapeExtents {x=600}}
 concurrencyOverrides {resolutionRule=PreventNew}
}
