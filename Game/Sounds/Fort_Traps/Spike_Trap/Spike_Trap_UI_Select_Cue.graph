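# Sound cue for the UI selection sound of the spike trap.
# A single wave player feeds the Spike_Trap_UI_Select wave straight to the output, with no randomization or modulation.
# It plays at reduced volume through the non-critical in-game UI sound class, with a very large max distance.
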SoundCue /Game/Sounds/Fort_Traps/Spike_Trap/Spike_Trap_UI_Select_Cue
root=WavePlayer_0
nodes [{id=WavePlayer_0, properties {soundWaveAssetPtr=&/Spike_Trap_UI_Select}}]
properties {maxDistance=2097152, soundClassObject=&Fort_UI_InGame_NonCritical, volumeMultiplier=0.3}
