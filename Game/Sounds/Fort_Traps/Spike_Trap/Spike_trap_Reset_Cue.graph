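# Sound cue for the floor spike trap's reset, playing "Ready" waves through the Fort_Traps sound class.
# A quality-level root sends its first two slots to an equally weighted random pick of two waves, and its last two to a single wave.
# Each branch passes through its own modulator for variation.
# Trap attenuation handles spatial falloff, and a concurrency override prevents new instances once the limit is reached.
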SoundCue /Game/Sounds/Fort_Traps/Spike_Trap/Spike_trap_Reset_Cue
root=QualityLevel_0
nodes [
 {id=Modulator_0, children [WavePlayer_0]}
 {id=Modulator_8, children [Random_16]}
 {id=QualityLevel_0, children [Modulator_8,Modulator_8,Modulator_0,Modulator_0]}
 {id=Random_16, children [WavePlayer_60, WavePlayer_61], properties {weights [1,1]}}
 {id=WavePlayer_0, properties {soundWaveAssetPtr=&/FloorTrap_Spikes_Ready_002}}
 {id=WavePlayer_60, properties {soundWaveAssetPtr=&/FloorTrap_Spikes_Ready_001}}
 {id=WavePlayer_61, properties {soundWaveAssetPtr=&FloorTrap_Spikes_Ready_002}}
]
properties {
 attenuationSettings=&Fort_Traps_Att
 bOverrideConcurrency=true
 maxDistance=6000
 soundClassObject=&Fort_Traps
 volumeMultiplier=0.5
 attenuationOverrides {falloffDistance=2400, attenuationShapeExtents {x=600}}
 concurrencyOverrides {resolutionRule=PreventNew}
}
